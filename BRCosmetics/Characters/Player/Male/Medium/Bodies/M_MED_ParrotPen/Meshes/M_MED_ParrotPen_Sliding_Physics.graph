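# Sliding-variant physics asset for the skirt of the male medium ParrotPen player cosmetic body.
# Seven skirt chains (A–G) each link a base bone and four dyn segments through per-joint swing and twist limits, with twist locked on many links.
# Capsules also cover both thighs and a very large root volume offset downward.
# Skirt bases and thighs constrain to root, all angularly locked except base_skirt_A and thigh_r, which are free.
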PhysicsAsset /BRCosmetics/Characters/Player/Male/Medium/Bodies/M_MED_ParrotPen/Meshes/M_MED_ParrotPen_Sliding_Physics
bodies {
 base_skirt_A [{shapeType=Capsule, length=1, radius=1, shapeName=base_skirt_A_capsule}]
 base_skirt_B [{shapeType=Capsule, length=5, radius=3, shapeName=base_skirt_B_capsule}]
 base_skirt_C [{shapeType=Capsule, length=5, radius=3, shapeName=base_skirt_C_capsule}]
 base_skirt_D [{shapeType=Capsule, length=5, radius=3, shapeName=base_skirt_D_capsule}]
 base_skirt_E [{shapeType=Capsule, length=5, radius=3, shapeName=base_skirt_E_capsule}]
 base_skirt_F [{shapeType=Capsule, length=5, radius=3, shapeName=base_skirt_F_capsule}]
 base_skirt_G [{shapeType=Capsule, length=5, radius=3, shapeName=base_skirt_G_capsule}]
 dyn_skirt_A_01 [{shapeType=Capsule, length=1, radius=1.5, shapeName=dyn_skirt_A_01_capsule, center=<5.8217 0 0>, rotation=<rot 90 0 0>}]
 dyn_skirt_A_02 [{shapeType=Capsule, length=2.5, radius=2, shapeName=dyn_skirt_A_02_capsule, center=<5.4313 0 0.8648>, rotation=<rot 90 0 0>}]
 dyn_skirt_A_03 [{shapeType=Capsule, length=2.5, radius=3.5, shapeName=dyn_skirt_A_03_capsule, center=<4.6947 0 0>, rotation=<rot 90 0 0>}]
 dyn_skirt_A_04 [{shapeType=Capsule, length=4, radius=4.5, shapeName=dyn_skirt_A_04_capsule, center=<7.278 0 0>, rotation=<rot 90 0 0>}]
 dyn_skirt_B_01 [{shapeType=Capsule, length=5, radius=3, shapeName=dyn_skirt_B_01_capsule, center=<6.4853 0 0>}]
 dyn_skirt_B_02 [{shapeType=Capsule, length=6.5, radius=2.5, shapeName=dyn_skirt_B_02_capsule, center=<9 0 0>}]
 dyn_skirt_B_03 [{shapeType=Capsule, length=11, radius=4, shapeName=dyn_skirt_B_03_capsule, center=<7.0288 0 0>, rotation=<rot 0 0 -5.625>}]
 dyn_skirt_B_04 [{shapeType=Capsule, length=11, radius=4, shapeName=dyn_skirt_B_04_capsule, center=<12.2864 -1.343 -0.3364>, rotation=<rot 0.0001 0 -14.0625>}]
 dyn_skirt_C_01 [{shapeType=Capsule, length=5, radius=3, shapeName=dyn_skirt_C_01_capsule, center=<6.4853 0 0>}]
 dyn_skirt_C_02 [{shapeType=Capsule, length=6.5, radius=3, shapeName=dyn_skirt_C_02_capsule, center=<7.2451 0 0>}]
 dyn_skirt_C_03 [{shapeType=Capsule, length=7.5, radius=4, shapeName=dyn_skirt_C_03_capsule, center=<7.8078 0 0>}]
 dyn_skirt_C_04 [{shapeType=Capsule, length=8.5, radius=4.5, shapeName=dyn_skirt_C_04_capsule, center=<12.2864 0 0>}]
 dyn_skirt_D_01 [{shapeType=Capsule, length=5, radius=3, shapeName=dyn_skirt_D_01_capsule, center=<6.4853 0 0>}]
 dyn_skirt_D_02 [{shapeType=Capsule, length=7, radius=3, shapeName=dyn_skirt_D_02_capsule, center=<7.0202 0 0>}]
 dyn_skirt_D_03 [{shapeType=Capsule, length=6, radius=4, shapeName=dyn_skirt_D_03_capsule, center=<7.8078 0 0>}]
 dyn_skirt_D_04 [{shapeType=Capsule, length=7, radius=4.5, shapeName=dyn_skirt_D_04_capsule, center=<12.2864 0 0>}]
 dyn_skirt_E_01 [{shapeType=Capsule, length=5, radius=3, shapeName=dyn_skirt_E_01_capsule, center=<6.4853 0 0>}]
 dyn_skirt_E_02 [{shapeType=Capsule, length=6.5, radius=3, shapeName=dyn_skirt_E_02_capsule, center=<7.1526 -1.7603 0>}]
 dyn_skirt_E_03 [{shapeType=Capsule, length=7.5, radius=4, shapeName=dyn_skirt_E_03_capsule, center=<7.8078 0 0>}]
 dyn_skirt_E_04 [{shapeType=Capsule, length=5, radius=4.5, shapeName=dyn_skirt_E_04_capsule, center=<12.2864 0 0>}]
 dyn_skirt_F_01 [{shapeType=Capsule, length=6, radius=2, shapeName=dyn_skirt_F_01_capsule, center=<9.3242 0 0>}]
 dyn_skirt_F_02 [{shapeType=Capsule, length=5.5, radius=2.5, shapeName=dyn_skirt_F_02_capsule, center=<9 0 0>}]
 dyn_skirt_F_03 [{shapeType=Capsule, length=9.5, radius=3.5, shapeName=dyn_skirt_F_03_capsule, center=<8.8714 -2.0978 0>}]
 dyn_skirt_F_04 [{shapeType=Capsule, length=9.5, radius=4, shapeName=dyn_skirt_F_04_capsule, center=<12.2864 -7.4029 0>}]
 dyn_skirt_G_01 [{shapeType=Capsule, length=1.5, radius=1.5, shapeName=dyn_skirt_G_01_capsule, center=<5.7933 0 0>, rotation=<rot 90 0 0>}]
 dyn_skirt_G_02 [{shapeType=Capsule, length=1.5, radius=3.5, shapeName=dyn_skirt_G_02_capsule, center=<4.9155 0 0>, rotation=<rot 90 0 0>}]
 dyn_skirt_G_03 [{shapeType=Capsule, length=3, radius=4, shapeName=dyn_skirt_G_03_capsule, center=<6.7321 0 0>, rotation=<rot 90 0 0>}]
 dyn_skirt_G_04 [{shapeType=Capsule, length=2, radius=4.5, shapeName=dyn_skirt_G_04_capsule, center=<9.1641 -1.5036 0>, rotation=<rot 90 0 0>}]
 root [{shapeType=Capsule, length=4, radius=70.0355, shapeName=root_capsule, center=<0 -6.573 -57.7941>}]
 thigh_l [{
   shapeType=Capsule
   length=35.5
   radius=10.5
   shapeName=thigh_l_capsule
   center=<-20.1244 -3.3918 1.4817>
   rotation=<rot 84.375 179.9996 179.9998>
  }]
 thigh_r [{
   shapeType=Capsule
   length=35.5
   radius=10.5
   shapeName=thigh_l_capsule
   center=<20.115 3.3918 -1.3862>
   rotation=<rot 84.375 179.9998 -0.0001>
  }]
}
constraints [
 {
  bone1Name=base_skirt_A
  bone2Name=root
  swing1LimitDegrees=45
  swing1Motion=Free
  swing2LimitDegrees=45
  swing2Motion=Free
  twistLimitDegrees=45
  twistMotion=Free
 }
 {bone1Name=dyn_skirt_A_01, bone2Name=base_skirt_A, swing1LimitDegrees=15, swing2LimitDegrees=15, twistLimitDegrees=5}
 {bone1Name=dyn_skirt_A_02, bone2Name=dyn_skirt_A_01, swing1LimitDegrees=30, swing2LimitDegrees=15, twistLimitDegrees=1}
 {bone1Name=dyn_skirt_A_03, bone2Name=dyn_skirt_A_02, swing1LimitDegrees=20, swing2LimitDegrees=15, twistLimitDegrees=5}
 {bone1Name=dyn_skirt_A_04, bone2Name=dyn_skirt_A_03, swing1LimitDegrees=15, swing2LimitDegrees=15, twistLimitDegrees=5, twistMotion=Locked}
 {
  bone1Name=base_skirt_B
  bone2Name=root
  swing1LimitDegrees=45
  swing1Motion=Locked
  swing2LimitDegrees=45
  swing2Motion=Locked
  twistLimitDegrees=45
  twistMotion=Locked
 }
 {bone1Name=dyn_skirt_B_01, bone2Name=base_skirt_B, swing1LimitDegrees=45, swing2LimitDegrees=10, twistLimitDegrees=5, twistMotion=Locked}
 {bone1Name=dyn_skirt_B_02, bone2Name=dyn_skirt_B_01, swing1LimitDegrees=35, swing2LimitDegrees=20, twistLimitDegrees=5}
 {bone1Name=dyn_skirt_B_03, bone2Name=dyn_skirt_B_02, swing1LimitDegrees=25, swing2LimitDegrees=20, twistLimitDegrees=5, twistMotion=Locked}
 {bone1Name=dyn_skirt_B_04, bone2Name=dyn_skirt_B_03, swing1LimitDegrees=15, swing2LimitDegrees=15, twistLimitDegrees=5, twistMotion=Locked}
 {
  bone1Name=base_skirt_C
  bone2Name=root
  swing1LimitDegrees=45
  swing1Motion=Locked
  swing2LimitDegrees=45
  swing2Motion=Locked
  twistLimitDegrees=45
  twistMotion=Locked
 }
 {bone1Name=dyn_skirt_C_01, bone2Name=base_skirt_C, swing1LimitDegrees=5, swing2LimitDegrees=10, twistLimitDegrees=5, twistMotion=Locked}
 {bone1Name=dyn_skirt_C_02, bone2Name=dyn_skirt_C_01, swing1LimitDegrees=10, swing2LimitDegrees=10, twistLimitDegrees=5}
 {bone1Name=dyn_skirt_C_03, bone2Name=dyn_skirt_C_02, swing1LimitDegrees=25, swing2LimitDegrees=5, twistLimitDegrees=5, twistMotion=Locked}
 {bone1Name=dyn_skirt_C_04, bone2Name=dyn_skirt_C_03, swing1LimitDegrees=45, swing2LimitDegrees=1, twistLimitDegrees=5, twistMotion=Locked}
 {
  bone1Name=base_skirt_D
  bone2Name=root
  swing1LimitDegrees=45
  swing1Motion=Locked
  swing2LimitDegrees=45
  swing2Motion=Locked
  twistLimitDegrees=45
  twistMotion=Locked
 }
 {bone1Name=dyn_skirt_D_01, bone2Name=base_skirt_D, swing1LimitDegrees=5, swing2LimitDegrees=1, twistLimitDegrees=1}
 {bone1Name=dyn_skirt_D_02, bone2Name=dyn_skirt_D_01, swing1LimitDegrees=10, swing2LimitDegrees=1, twistLimitDegrees=5}
 {bone1Name=dyn_skirt_D_03, bone2Name=dyn_skirt_D_02, swing1LimitDegrees=25, swing2LimitDegrees=1, twistLimitDegrees=5, twistMotion=Locked}
 {bone1Name=dyn_skirt_D_04, bone2Name=dyn_skirt_D_03, swing1LimitDegrees=45, swing2LimitDegrees=1, twistLimitDegrees=5, twistMotion=Locked}
 {
  bone1Name=base_skirt_E
  bone2Name=root
  swing1LimitDegrees=45
  swing1Motion=Locked
  swing2LimitDegrees=45
  swing2Motion=Locked
  twistLimitDegrees=45
  twistMotion=Locked
 }
 {bone1Name=dyn_skirt_E_01, bone2Name=base_skirt_E, swing1LimitDegrees=5, swing2LimitDegrees=10, twistLimitDegrees=5, twistMotion=Locked}
 {bone1Name=dyn_skirt_E_02, bone2Name=dyn_skirt_E_01, swing1LimitDegrees=10, swing2LimitDegrees=10, twistLimitDegrees=5}
 {bone1Name=dyn_skirt_E_03, bone2Name=dyn_skirt_E_02, swing1LimitDegrees=25, swing2LimitDegrees=5, twistLimitDegrees=5, twistMotion=Locked}
 {bone1Name=dyn_skirt_E_04, bone2Name=dyn_skirt_E_03, swing1LimitDegrees=45, swing2LimitDegrees=1, twistLimitDegrees=5, twistMotion=Locked}
 {
  bone1Name=base_skirt_F
  bone2Name=root
  swing1LimitDegrees=45
  swing1Motion=Locked
  swing2LimitDegrees=45
  swing2Motion=Locked
  twistLimitDegrees=45
  twistMotion=Locked
 }
 {bone1Name=dyn_skirt_F_01, bone2Name=base_skirt_F, swing1LimitDegrees=45, swing2LimitDegrees=10, twistLimitDegrees=5, twistMotion=Locked}
 {bone1Name=dyn_skirt_F_02, bone2Name=dyn_skirt_F_01, swing1LimitDegrees=35, swing2LimitDegrees=20, twistLimitDegrees=5}
 {bone1Name=dyn_skirt_F_03, bone2Name=dyn_skirt_F_02, swing1LimitDegrees=25, swing2LimitDegrees=20, twistLimitDegrees=5, twistMotion=Locked}
 {bone1Name=dyn_skirt_F_04, bone2Name=dyn_skirt_F_03, swing1LimitDegrees=25, swing2LimitDegrees=15, twistLimitDegrees=5, twistMotion=Locked}
 {
  bone1Name=base_skirt_G
  bone2Name=root
  swing1LimitDegrees=45
  swing1Motion=Locked
  swing2LimitDegrees=45
  swing2Motion=Locked
  twistLimitDegrees=45
  twistMotion=Locked
 }
 {bone1Name=dyn_skirt_G_01, bone2Name=base_skirt_G, swing1LimitDegrees=65, swing2LimitDegrees=35, twistLimitDegrees=5}
 {bone1Name=dyn_skirt_G_02, bone2Name=dyn_skirt_G_01, swing1LimitDegrees=30, swing2LimitDegrees=15, twistLimitDegrees=5, twistMotion=Locked}
 {bone1Name=dyn_skirt_G_03, bone2Name=dyn_skirt_G_02, swing1LimitDegrees=20, swing2LimitDegrees=15, twistLimitDegrees=5}
 {bone1Name=dyn_skirt_G_04, bone2Name=dyn_skirt_G_03, swing1LimitDegrees=15, swing2LimitDegrees=15, twistLimitDegrees=5, twistMotion=Locked}
 {
  bone1Name=thigh_r
  bone2Name=root
  swing1LimitDegrees=45
  swing1Motion=Free
  swing2LimitDegrees=45
  swing2Motion=Free
  twistLimitDegrees=45
  twistMotion=Free
 }
 {
  bone1Name=thigh_l
  bone2Name=root
  swing1LimitDegrees=45
  swing1Motion=Locked
  swing2LimitDegrees=45
  swing2Motion=Locked
  twistLimitDegrees=45
  twistMotion=Locked
 }
]
