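# Physics asset for the strap mesh of the M_MED_ParrotPen male player cosmetic body, describing a two-bone strap chain.
# A small capsule sits on base_strap, and a larger capsule on dyn_strap is offset along that bone's X axis.
# A single constraint joins dyn_strap to base_strap with narrow, asymmetric swing limits and locked twist, so the strap can sway without twisting.
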PhysicsAsset /BRCosmetics/Characters/Player/Male/Medium/Bodies/M_MED_ParrotPen/Meshes/M_MED_ParrotPen_Strap_Physics
bodies {
 base_strap [{shapeType=Capsule, length=0.505, radius=0.505, shapeName=base_strap_capsule, rotation=<rot 0 -90 0>}]
 dyn_strap [{shapeType=Capsule, length=2, radius=1, shapeName=dyn_strap_capsule, center=<7.9979 0 0>, rotation=<rot 90 0 0>}]
}
constraints [{bone1Name=dyn_strap, bone2Name=base_strap, swing1LimitDegrees=10, swing2LimitDegrees=25, twistLimitDegrees=45, twistMotion=Locked}]
