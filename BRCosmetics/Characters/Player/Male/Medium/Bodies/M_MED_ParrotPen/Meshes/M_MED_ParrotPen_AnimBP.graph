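# Animation Blueprint for the male medium ParrotPen cosmetic body, adding secondary motion to skirt, strap and shoulder gear.
# Parent mesh pose feeds a deform Control Rig, bone constraints and strap physics, then skirt RigidBody simulation with state-specific offsets and a separate sliding setup.
# Look At poses replace skirt physics when crouching or sliding with rigid bodies disabled.
# EventGraph updates state values and drives material WPO cloth from leg motion and skydiving wind.
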AnimBlueprint /BRCosmetics/Characters/Player/Male/Medium/Bodies/M_MED_ParrotPen/Meshes/M_MED_ParrotPen_AnimBP
target_skeleton=&/M_MED_ParrotPen_Skeleton
graphs [
 AnimGraph {nodes [
   n1 = Get bIsInFrontEnd()
   n2 = RigidBody!(ComponentPose=Local To Component(LocalPose=Use cached pose 'Strap'()), Alpha, OverrideWorldGravity:Vector="0,0,-980") -> (Pose) {
    bClampLinearTranslationLimitToRefPose=true
    bOverrideWorldGravity=true
    overridePhysicsAsset=&/M_MED_ParrotPen_Physics
    simulationSpace=BaseBoneSpace
    alphaBoolBlend {blendOption=Linear}
    baseBoneRef {boneName=pelvis}
    componentAppliedLinearAccClamp=<3000 3000 3000>
    componentLinearAccScale=<0.2 0.1 0.1>
    componentLinearVelScale=<0.2 0.1 0.1>
    overrideWorldGravity {z=-980}
    simSpaceSettings {
     maxAngularAcceleration=1002
     maxAngularVelocity=3
     maxLinearAcceleration=700
     maxLinearVelocity=500
     velocityScaleZ=0
     worldAlpha=0.5
     externalLinearDragV=<0.2 0.2>
    }
   }
   n3 = Get bIsCrouching()
   n4 = "Transform (Modify) Bone - Bone: dyn_skirt_G_02"!(ComponentPose=@n2, bAlphaBoolEnabled:bool=@n3, Rotation="3.908,3.187,6.877") -> (Pose) {
    alphaInputType=Bool
    rotationMode=BMM_Additive
    translationMode=BMM_Additive
    alphaBoolBlend {blendOption=Linear}
    boneToModify {boneName=dyn_skirt_G_02}
    rotation=<rot 3.908 3.187 6.877>
    translation=<-0.5218 2.0209 0.4065>
   }
   n5 = "Transform (Modify) Bone - Bone: dyn_skirt_G_04"!(ComponentPose=@n4, bAlphaBoolEnabled:bool=@n3, Rotation="4.935,1.445,14.159") -> (Pose) {
    alphaInputType=Bool
    rotationMode=BMM_Additive
    translationMode=BMM_Additive
    alphaBoolBlend {blendOption=Linear}
    boneToModify {boneName=dyn_skirt_G_04}
    rotation=<rot 4.935 1.445 14.159>
    translation=<0.3691 -3.8841 -0.2557>
   }
   n6 = Get bIsDBNO()
   n7 = "Transform (Modify) Bone - Bone: dyn_skirt_C_02"!(ComponentPose=@n5, bAlphaBoolEnabled:bool=@n6, Rotation="0,0,-10") -> (Pose) {
    alphaInputType=Bool
    rotationMode=BMM_Additive
    translationMode=BMM_Additive
    alphaBoolBlend {blendOption=Linear}
    boneToModify {boneName=dyn_skirt_C_02}
    rotation {roll=-10}
    translation=<-0.5218 2.0209 0.4065>
   }
   n8 = "Transform (Modify) Bone - Bone: dyn_skirt_D_02"!(ComponentPose=@n7, bAlphaBoolEnabled:bool=@n6, Rotation="0,0,-10") -> (Pose) {
    alphaInputType=Bool
    rotationMode=BMM_Additive
    translationMode=BMM_Additive
    alphaBoolBlend {blendOption=Linear}
    boneToModify {boneName=dyn_skirt_D_02}
    rotation {roll=-10}
    translation=<-0.5218 2.0209 0.4065>
   }
   n9 = "Transform (Modify) Bone - Bone: dyn_skirt_E_02"!(ComponentPose=@n8, bAlphaBoolEnabled:bool=@n6, Rotation="0,0,-10") -> (Pose) {
    alphaInputType=Bool
    rotationMode=BMM_Additive
    translationMode=BMM_Additive
    alphaBoolBlend {blendOption=Linear}
    boneToModify {boneName=dyn_skirt_E_02}
    rotation {roll=-10}
    translation=<-0.5218 2.0209 0.4065>
   }
   n10 = "Transform (Modify) Bone - Bone: base_skirt_C"!(ComponentPose=@n9, bAlphaBoolEnabled:bool=@n1, Rotation="7,0,7") -> (Pose) {
    alphaInputType=Bool
    rotationMode=BMM_Additive
    translationMode=BMM_Additive
    alphaBoolBlend {blendOption=Linear}
    boneToModify {boneName=base_skirt_C}
    rotation {pitch=7, roll=7}
   }
   n11 = "Transform (Modify) Bone - Bone: base_skirt_E"!(ComponentPose=@n10, bAlphaBoolEnabled:bool=@n1, Rotation="3,0,3") -> (Pose) {
    alphaInputType=Bool
    rotationMode=BMM_Additive
    translationMode=BMM_Additive
    alphaBoolBlend {blendOption=Linear}
    boneToModify {boneName=base_skirt_E}
    rotation {pitch=3, roll=3}
   }
   n12 = Look At - Bone: dyn_skirt_A_01(ComponentPose=Local To Component(LocalPose=Use cached pose 'Strap'()), Alpha=1) {
    alphaBoolBlend {blendOption=Linear}
    boneToModify {boneName=dyn_skirt_A_01}
    lookAtLocation=<65.3038 19.085 0>
    lookAtTarget {boneReference {boneName=root}}
    lookAt_Axis {axis=<1 0>}
   }
   n13 = Look At - Bone: base_skirt_B(ComponentPose=@n12, Alpha=1) {
    alphaBoolBlend {blendOption=Linear}
    boneToModify {boneName=base_skirt_B}
    lookAtLocation=<63.1667 -33.5361>
    lookAtTarget {boneReference {boneName=root}}
    lookAt_Axis {axis=<1 0>}
   }
   n14 = Look At - Bone: base_skirt_F(ComponentPose=@n13, Alpha=1) {
    alphaBoolBlend {blendOption=Linear}
    boneToModify {boneName=base_skirt_F}
    lookAtLocation=<-71.5141 -36.7108>
    lookAtTarget {boneReference {boneName=root}}
    lookAt_Axis {axis=<1 0>}
   }
   n15 = Look At - Bone: base_skirt_G(ComponentPose=@n14, Alpha=1) {
    alphaBoolBlend {blendOption=Linear}
    boneToModify {boneName=base_skirt_G}
    lookAtLocation=<-72.1311 1.9279 0>
    lookAtTarget {boneReference {boneName=root}}
    lookAt_Axis {axis=<1 0>}
   }
   n16 = "Transform (Modify) Bone - Bone: base_skirt_E"!(ComponentPose=@n15, bAlphaBoolEnabled=True, Translation:Vector="0.224,-3.620,7.298") -> (Pose) {
    alphaInputType=Bool
    rotationMode=BMM_Additive
    translationMode=BMM_Additive
    alphaBoolBlend {blendOption=Linear}
    boneToModify {boneName=base_skirt_E}
    translation=<0.224 -3.62 7.298>
   }
   n17 = RigidBody(ComponentPose=@n16, OverrideWorldGravity="0,0,-980") {
    bClampLinearTranslationLimitToRefPose=true
    bOverrideWorldGravity=true
    overridePhysicsAsset=&/M_MED_ParrotPen_Sliding_Physics
    simulationSpace=BaseBoneSpace
    alphaBoolBlend {blendOption=Linear}
    baseBoneRef {boneName=pelvis}
    componentAppliedLinearAccClamp=<3000 3000 3000>
    componentLinearAccScale=<0.3 0.2 0.2>
    componentLinearVelScale=<0.3 0.2 0.2>
    overrideWorldGravity {z=-980}
    simSpaceSettings {
     maxAngularAcceleration=1002
     maxAngularVelocity=3
     maxLinearAcceleration=700
     maxLinearVelocity=500
     velocityScaleZ=0
     worldAlpha=0.5
     externalLinearDragV=<0.2 0.2>
    }
   }
   n18 = Blend Poses by bool(
    BlendPose_0=Component To Local(ComponentPose=@n17)
    BlendPose_1=Component To Local(ComponentPose=@n11)
    BlendTime_0=0.5
    bActiveValue=Get bIsSliding()
   ) {blendType=Linear}
   n19 = Look At - Bone: dyn_skirt_A_01(ComponentPose=Local To Component(LocalPose=Use cached pose 'Strap'()), Alpha=1) {
    alphaBoolBlend {blendOption=Linear}
    boneToModify {boneName=dyn_skirt_A_01}
    lookAtLocation=<13.1809 5.3193 -10.2385>
    lookAtTarget {boneReference {boneName=deform_knee_out_l}}
    lookAt_Axis {axis=<1 0>}
   }
   n20 = Look At - Bone: base_skirt_B(ComponentPose=@n19, Alpha=1) {
    alphaBoolBlend {blendOption=Linear}
    boneToModify {boneName=base_skirt_B}
    lookAtLocation=<47.1047 -18.7547 0>
    lookAtTarget {boneReference {boneName=root}}
    lookAt_Axis {axis=<1 0>}
   }
   n21 = Look At - Bone: base_skirt_C(ComponentPose=@n20, Alpha=1) {
    alphaBoolBlend {blendOption=Linear}
    boneToModify {boneName=base_skirt_C}
    lookAtLocation=<40.8918 -56.769 0>
    lookAtTarget {boneReference {boneName=root}}
    lookAt_Axis {axis=<1 0>}
   }
   n22 = Look At - Bone: base_skirt_D(ComponentPose=@n21, Alpha=1) {
    alphaBoolBlend {blendOption=Linear}
    boneToModify {boneName=base_skirt_D}
    lookAtLocation=<11.8422 -88.466>
    lookAtTarget {boneReference {boneName=root}}
    lookAt_Axis {axis=<1 0>}
   }
   n23 = Look At - Bone: base_skirt_E(ComponentPose=@n22, Alpha=1) {
    alphaBoolBlend {blendOption=Linear}
    boneToModify {boneName=base_skirt_E}
    lookAtLocation=<-38.4206 -63.6012 0>
    lookAtTarget {boneReference {boneName=root}}
    lookAt_Axis {axis=<1 0>}
   }
   n24 = Look At - Bone: base_skirt_F(ComponentPose=@n23, Alpha=1) {
    alphaBoolBlend {blendOption=Linear}
    boneToModify {boneName=base_skirt_F}
    lookAtLocation=<-34.092 -40.8118 0>
    lookAtTarget {boneReference {boneName=root}}
    lookAt_Axis {axis=<1 0>}
   }
   n25 = Look At - Bone: base_skirt_G(ComponentPose=@n24, Alpha=1) {
    alphaBoolBlend {blendOption=Linear}
    boneToModify {boneName=base_skirt_G}
    lookAtLocation=<-1.669 -8.8662 13.0197>
    lookAtTarget {boneReference {boneName=deform_knee_out_r}}
    lookAt_Axis {axis=<1 0>}
   }
   n26 = Component To Local(ComponentPose=@n25)
   n27 = Get CrouchRBdisabled() -> (CrouchRBdisabled:bool)
   n28 = Look At - Bone: dyn_skirt_A_01(ComponentPose=Local To Component(LocalPose=Use cached pose 'Strap'()), Alpha=1) {
    alphaBoolBlend {blendOption=Linear}
    boneToModify {boneName=dyn_skirt_A_01}
    lookAtLocation=<65.3038 19.085 0>
    lookAtTarget {boneReference {boneName=root}}
    lookAt_Axis {axis=<1 0>}
   }
   n29 = Look At - Bone: base_skirt_B(ComponentPose=@n28, Alpha=1) {
    alphaBoolBlend {blendOption=Linear}
    boneToModify {boneName=base_skirt_B}
    lookAtLocation=<63.1667 -33.5361 0>
    lookAtTarget {boneReference {boneName=root}}
    lookAt_Axis {axis=<1 0>}
   }
   n30 = Look At - Bone: base_skirt_C(ComponentPose=@n29, Alpha=1) {
    alphaBoolBlend {blendOption=Linear}
    boneToModify {boneName=base_skirt_C}
    lookAtLocation=<40.8918 -56.769 0>
    lookAtTarget {boneReference {boneName=root}}
    lookAt_Axis {axis=<1 0>}
   }
   n31 = Look At - Bone: base_skirt_D(ComponentPose=@n30, Alpha=1) {
    alphaBoolBlend {blendOption=Linear}
    boneToModify {boneName=base_skirt_D}
    lookAtLocation=<0 -77.3673 0>
    lookAtTarget {boneReference {boneName=root}}
    lookAt_Axis {axis=<1 0>}
   }
   n32 = Look At - Bone: base_skirt_E(ComponentPose=@n31, Alpha=1) {
    alphaBoolBlend {blendOption=Linear}
    boneToModify {boneName=base_skirt_E}
    lookAtLocation=<-46.4407 -67.7825 0>
    lookAtTarget {boneReference {boneName=root}}
    lookAt_Axis {axis=<1 0>}
   }
   n33 = Look At - Bone: base_skirt_F(ComponentPose=@n32, Alpha=1) {
    alphaBoolBlend {blendOption=Linear}
    boneToModify {boneName=base_skirt_F}
    lookAtLocation=<-71.5141 -36.7108 0>
    lookAtTarget {boneReference {boneName=root}}
    lookAt_Axis {axis=<1 0>}
   }
   n34 = Look At - Bone: base_skirt_G(ComponentPose=@n33, Alpha=1) {
    alphaBoolBlend {blendOption=Linear}
    boneToModify {boneName=base_skirt_G}
    lookAtLocation=<-72.1311 1.9279 0>
    lookAtTarget {boneReference {boneName=root}}
    lookAt_Axis {axis=<1 0>}
   }
   n35 = "Transform (Modify) Bone - Bone: base_skirt_E"(ComponentPose=@n34, bAlphaBoolEnabled=True, Translation="0.224,-3.620,8.252") {
    alphaInputType=Bool
    rotationMode=BMM_Additive
    translationMode=BMM_Additive
    alphaBoolBlend {blendOption=Linear}
    boneToModify {boneName=base_skirt_E}
    translation=<0.224 -3.62 8.252>
   }
   Output Pose(Result=Blend Poses by bool(
     BlendPose_0=Component To Local(ComponentPose=@n35)
     BlendPose_1=Blend Poses by bool(BlendPose_0=@n26, BlendPose_1=@n18, bActiveValue=@n27) {blendType=Linear}
     bActiveValue=Get SlidingRBdisabled() -> (SlidingRBdisabled:bool)
    ) {blendType=Linear}) {layerGroup=None}
   n36 = Control Rig!(Source=Copy Pose From Mesh() {bUseAttachedParent=true}, Alpha, isMoving2D:double=Get Moving2d() -> (Moving2d:double)) -> (Pose) {
    alphaBoolBlend {blendOption=Linear}
    controlRigAssetReference {blueprintRigClass=&/M_MED_ParrotPen_Deform_ControlRig.M_MED_ParrotPen_Deform_ControlRig_C}
   }
   Save cached pose 'Start'(Pose=@n36) {cacheName=Start}
   n37 = "Constraint - base_gear_l (PP)"!(ComponentPose=Local To Component(LocalPose=Use cached pose 'Start'()), Alpha=1, ConstraintWeights_0=0.9, ConstraintWeights_1:float=0.1) -> (Pose) {
    alphaBoolBlend {blendOption=Linear}
    boneToModify {boneName=base_gear_l}
    constraintSetup [{targetBone {boneName=clavicle_l}}, {targetBone {boneName=upperarm_l}}]
    constraintWeights [1,1]
   }
   n38 = "Constraint - base_gear_r (PP)"!(ComponentPose=@n37, Alpha=1, ConstraintWeights_0=0.9, ConstraintWeights_1:float=0.1) -> (Pose) {
    alphaBoolBlend {blendOption=Linear}
    boneToModify {boneName=base_gear_r}
    constraintSetup [{targetBone {boneName=clavicle_r}}, {targetBone {boneName=upperarm_r}}]
    constraintWeights [1,1]
   }
   n39 = Get ParachuteOpened() -> (ParachuteOpened:double)
   n40 = "Transform (Modify) Bone - Bone: dyn_gear_l"!(ComponentPose=@n38, Alpha=@n39, Rotation="-5,-10,0") -> (Pose) {
    rotationMode=BMM_Additive
    rotationSpace=BCS_BoneSpace
    alphaBoolBlend {blendOption=Linear}
    boneToModify {boneName=dyn_gear_l}
    rotation {pitch=-5, yaw=-10}
   }
   n41 = "Transform (Modify) Bone - Bone: dyn_gear_r"!(ComponentPose=@n40, Alpha=@n39, Rotation="5,10,0") -> (Pose) {
    rotationMode=BMM_Additive
    rotationSpace=BCS_BoneSpace
    alphaBoolBlend {blendOption=Linear}
    boneToModify {boneName=dyn_gear_r}
    rotation {pitch=5, yaw=10}
   }
   Save cached pose 'Gears'(Pose=Component To Local(ComponentPose=@n41)) {cacheName=Gears}
   n42 = "Constraint - base_strap (PPPPP)"!(
    ComponentPose=Local To Component(LocalPose=Use cached pose 'Gears'())
    Alpha=1
    ConstraintWeights_0=0.3275
    ConstraintWeights_1:float=0.2996
    ConstraintWeights_2:float=0.1872
    ConstraintWeights_3:float=0.1261
    ConstraintWeights_4:float=0.0596
   ) -> (Pose) {
    alphaBoolBlend {blendOption=Linear}
    boneToModify {boneName=base_strap}
    constraintSetup [
     {targetBone {boneName=spine_02}}
     {targetBone {boneName=spine_03}}
     {targetBone {boneName=spine_01}}
     {targetBone {boneName=pelvis}}
     {targetBone {boneName=spine_04}}
    ]
    constraintWeights [1,1,1,1,1]
   }
   n43 = RigidBody(ComponentPose=@n42, Alpha=Get isNotEmote() -> (isNotEmote:double), OverrideWorldGravity="0,0,-980") {
    bClampLinearTranslationLimitToRefPose=true
    bOverrideWorldGravity=true
    overridePhysicsAsset=&/M_MED_ParrotPen_Strap_Physics
    simulationSpace=BaseBoneSpace
    alphaBoolBlend {blendOption=Linear}
    baseBoneRef {boneName=pelvis}
    componentAppliedLinearAccClamp=<3000 3000 3000>
    componentLinearAccScale=<0.2 0.1 0.1>
    componentLinearVelScale=<0.2 0.1 0.1>
    overrideWorldGravity {z=-980}
    simSpaceSettings {
     maxAngularAcceleration=1002
     maxAngularVelocity=3
     maxLinearAcceleration=700
     maxLinearVelocity=500
     velocityScaleZ=0
     worldAlpha=0.5
     externalLinearDragV=<0.2 0.2>
    }
   }
   Save cached pose 'Strap'(Pose=Component To Local(ComponentPose=@n43)) {cacheName=Strap}
  ]}
 EventGraph {nodes [
   n1 = Event BlueprintUpdateAnimation()
   > n2 = Sequence!(execute) -> (then_0, then_1, then_2:Exec, then_3:Exec)
   > n3 = Gravity Override State Function DEPRECATED(CharacterPartAnimInstance=Self-Reference(), GameTimeInSeconds=GetGameTimeInSeconds(), GravityOverrideParams=Get Gravity Override Params())
   > Set GravityOverride(GravityOverride=@n3)
   n4 = Get bRigidBodyEnabled()
   Set CrouchRBdisabled(execute=@n2.then_1, CrouchRBdisabled:bool=AND Boolean(A=Get bIsCrouching(), B=NOT Boolean(A=@n4))) -> (Output_Get:bool)
   > Set Runing(Runing:double="To Float (Boolean)"(InBool=Get bIsSprinting())) -> (Output_Get:double)
   > Set ParachuteOpened(ParachuteOpened:double="To Float (Boolean)"(InBool=Get bIsParachuteOpened())) -> (Output_Get:double)
   n5 = Get bIsSliding()
   > Set Sliding(Sliding:double="To Float (Boolean)"(InBool=@n5)) -> (Output_Get:double)
   n6 = Get bRigidBodyEnabled()
   > Set SlidingRBdisabled(SlidingRBdisabled:bool=AND Boolean(A=@n5, B=NOT Boolean(A=@n6))) -> (Output_Get:bool)
   n7 = Get bIsPlayingEmote()
   > Set isNotEmote(isNotEmote:double="To Float (Boolean)"(InBool=NOT Boolean(A=@n7))) -> (Output_Get:double)
   > n8 = Branch(Condition=Get bIsCrouching())
   n9 = "To Float (Boolean)"(InBool=Get bIsMoving2D())
   > Set Moving2d(Moving2d:double=@n9) -> (Output_Get:double)
   Set Moving2d(execute=@n8.else, Moving2d=FInterpTo(Current=Get Moving2d(), Target=@n9, DeltaTime=@n1.DeltaTimeX, InterpSpeed=20))
   Set WPO Globals Function(
    execute=@n2.then_2
    Material=Get Mat Instance Dynamic()
    Overlay Material=Get Overlay Mat Instance Dynamic()
    Skeletal Mesh=Get SkelMesh()
    "Local Noise Transform  Location"=Get Local Noise Transform Location() -> (Local Noise Transform Location:Vector)
    Local Noise Offset=Get Local Noise Offset()
   )
   > Set WPO Regional Function!(
    execute
    Material:Material Instance Dynamic Object Reference=Get Mat Instance Dynamic()
    Overlay Material:Material Instance Dynamic Object Reference=Get Overlay Mat Instance Dynamic()
    Skeletal Mesh:Skeletal Mesh Component Object Reference=Get SkelMesh()
    Velocity Transform:Name=foot_r
    Velocity Interp Speed:double=Get Velocity Interp Speed()
    Collision Transform:Name=calf_r
    Collision Transform Axis:EAxis Enum
    Invert Collision Transform Axis:bool=false
    Delta Time X:double=@n1.DeltaTimeX
    Previous Velocity Transform Location:"Vector (by ref)"=Get R1 Velocity Transform Location()
    Previous Velocity Transform Velocity:"Vector (by ref)"=Get R1 Velocity Transform Velocity()
    Region:int=1
   ) -> (then)
   > Set WPO Regional Function(
    Material=Get Mat Instance Dynamic()
    Overlay Material=Get Overlay Mat Instance Dynamic()
    Skeletal Mesh=Get SkelMesh()
    Velocity Transform=foot_l
    Velocity Interp Speed=Get Velocity Interp Speed()
    Collision Transform=calf_l
    Invert Collision Transform Axis=true
    Delta Time X=@n1.DeltaTimeX
    Previous Velocity Transform Location=Get R2 Velocity Transform Location()
    Previous Velocity Transform Velocity=Get R2 Velocity Transform Velocity()
    Region=2
   )
   > n10 = Branch(Condition=Get bIsSkyDiving())
   > Set WPO Pho Wind Intensity(WPO Pho Wind Intensity=0.8)
   > n11 = Set WPO Pho Turbulence(WPO Pho Turbulence:double=1.5) -> (Output_Get:double)
   SetScalarParameterValue(
    execute=[@n13.then, @n11.then]
    self=[Get Overlay Mat Instance Dynamic(), Get Mat Instance Dynamic()]
    ParameterName=ClothSim_WindStrength
    Value=Get WPO Pho Wind Intensity()
   )
   n12 = Get WPO Pho Turbulence() -> (WPO Pho Turbulence:double)
   > SetScalarParameterValue(self=[Get Overlay Mat Instance Dynamic(), Get Mat Instance Dynamic()], ParameterName=ClothSim_R1_TurbulenceStrength, Value=@n12)
   > SetScalarParameterValue(self=[Get Overlay Mat Instance Dynamic(), Get Mat Instance Dynamic()], ParameterName=ClothSim_R2_TurbulenceStrength, Value=@n12)
   Set WPO Pho Wind Intensity(execute=@n10.else, WPO Pho Wind Intensity=0)
   > n13 = Set WPO Pho Turbulence(WPO Pho Turbulence=0.2)
   Event OnCharacterCustomizationLoaded()
   > n14 = Set SkelMesh(SkelMesh=GetPartSkeletalMeshComponent())
   > n15 = Get Dynamic Material Instance(SkeletalMeshComponent=@n14)
   > Set Mat Instance Dynamic(Mat Instance Dynamic=@n15)
   > n16 = GetOrCreateDynamicOverlayMaterial(MeshComponent=@n14, MaterialName=None)
   > Set Overlay Mat Instance Dynamic(Overlay Mat Instance Dynamic=@n16)
  ]}
]
variables [
 CrouchRBdisabled {type=boolean}
 Gravity Override Params {type=GravityOverrideParamsStruct, default {skydive_FloatingDown {gravityOverride=<0 -700 100>}}}
 GravityOverride {type=Vector}
 isNotEmote {type=number}
 Local Noise Offset {type=Vector}
 Local Noise Transform Location {type=Vector}
 Mat Instance Dynamic {type=MaterialInstanceDynamic}
 Moving2d {type=number}
 Overlay Mat Instance Dynamic {type=MaterialInstanceDynamic}
 ParachuteOpened {type=number}
 R1 Velocity Transform Location {type=Vector}
 R1 Velocity Transform Velocity {type=Vector}
 R2 Velocity Transform Location {type=Vector}
 R2 Velocity Transform Velocity {type=Vector}
 Runing {type=number}
 SkelMesh {type=SkeletalMeshComponent}
 Sliding {type=number}
 SlidingRBdisabled {type=boolean}
 thigh_rot_up_l {type=number}
 thigh_rot_up_r {type=number}
 Velocity Interp Speed {type=number}
 WPO Pho Turbulence {type=number}
 WPO Pho Wind Intensity {type=number}
]
