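# Physics asset for the face accessory part of the male medium ParrotPen cosmetic body, defining hair physics around the head.
# Small base capsules for left, right and central hair are rotationally locked to a zero-length, effectively spherical head capsule.
# Each side dyn capsule pairs with its own base under tight swing limits and zero twist limit.
# All four central dyn capsules attach directly to base_hair01 rather than chaining, with slight twist allowed.
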PhysicsAsset /BRCosmetics/Characters/Player/Male/Medium/Bodies/M_MED_ParrotPen/Meshes/Parts/M_MED_ParrotPen_FaceAcc_Physics
bodies {
 base_L_hair01 [{shapeType=Capsule, length=0.505, radius=0.505, shapeName=base_L_hair01_capsule, rotation=<rot 0 -90 0>}]
 base_L_hair02 [{shapeType=Capsule, length=0.505, radius=0.505, shapeName=base_L_hair02_capsule, rotation=<rot 0 -90 0>}]
 base_L_hair03 [{shapeType=Capsule, length=0.505, radius=0.505, shapeName=base_L_hair03_capsule, rotation=<rot 0 -90 0>}]
 base_L_hair04 [{shapeType=Capsule, length=0.505, radius=0.505, shapeName=base_L_hair04_capsule, rotation=<rot 0 -90 0>}]
 base_R_hair01 [{shapeType=Capsule, length=0.505, radius=0.505, shapeName=base_R_hair01_capsule, rotation=<rot 0 -90 0>}]
 base_R_hair02 [{shapeType=Capsule, length=0.505, radius=0.505, shapeName=base_R_hair02_capsule, rotation=<rot 0 -90 0>}]
 base_hair01 [{shapeType=Capsule, length=0.505, radius=0.505, shapeName=base_hair01_capsule, rotation=<rot 0 -90 0>}]
 dyn_L_hair01 [{shapeType=Capsule, length=1, radius=0.5, shapeName=dyn_L_hair01_capsule, center=<1.1916 0 0>, rotation=<rot 90 0 0>}]
 dyn_L_hair02 [{shapeType=Capsule, length=0.5, radius=0.5, shapeName=dyn_L_hair02_capsule, center=<1.2049 0 0>, rotation=<rot 90 0 0>}]
 dyn_L_hair03 [{shapeType=Capsule, length=1, radius=0.5, shapeName=dyn_L_hair03_capsule, center=<1.23 0 0>, rotation=<rot 90 0 0>}]
 dyn_L_hair04 [{shapeType=Capsule, length=1, radius=0.5, shapeName=dyn_L_hair04_capsule, center=<0.9268 0 0>, rotation=<rot 90 0 0>}]
 dyn_R_hair01 [{shapeType=Capsule, length=1, radius=0.5, shapeName=dyn_R_hair01_capsule, center=<1.1916 0 0>, rotation=<rot 90 0 0>}]
 dyn_R_hair02 [{shapeType=Capsule, length=1, radius=0.5, shapeName=dyn_R_hair02_capsule, center=<2 0 0>, rotation=<rot 90 0 0>}]
 dyn_hair01 [{shapeType=Capsule, length=1, radius=1, shapeName=dyn_hair01_capsule, center=<3.3366 0 0>, rotation=<rot 90 0 0>}]
 dyn_hair02 [{shapeType=Capsule, length=1, radius=1, shapeName=dyn_hair02_capsule, center=<3.3366 0 0>, rotation=<rot 90 0 0>}]
 dyn_hair03 [{shapeType=Capsule, length=1, radius=1, shapeName=dyn_hair03_capsule, center=<3.3366 0 0>, rotation=<rot 90 0 0>}]
 dyn_hair04 [{shapeType=Capsule, length=1, radius=1, shapeName=dyn_hair04_capsule, center=<3.3366 0 0>, rotation=<rot 90 0 0>}]
 head [{shapeType=Capsule, length=0, radius=7.5, shapeName=head_capsule, center=<6.6496 1.1357 0>, rotation=<rot 90 0 0>}]
}
constraints [
 {bone1Name=dyn_L_hair01, bone2Name=base_L_hair01, swing1LimitDegrees=10, swing2LimitDegrees=15, twistLimitDegrees=0}
 {bone1Name=dyn_L_hair02, bone2Name=base_L_hair02, swing1LimitDegrees=10, swing2LimitDegrees=5, twistLimitDegrees=0}
 {bone1Name=dyn_L_hair03, bone2Name=base_L_hair03, swing1LimitDegrees=10, swing2LimitDegrees=10, twistLimitDegrees=0}
 {bone1Name=dyn_L_hair04, bone2Name=base_L_hair04, swing1LimitDegrees=15, swing2LimitDegrees=20, twistLimitDegrees=0}
 {bone1Name=dyn_R_hair01, bone2Name=base_R_hair01, swing1LimitDegrees=10, swing2LimitDegrees=15, twistLimitDegrees=0}
 {bone1Name=dyn_R_hair02, bone2Name=base_R_hair02, swing1LimitDegrees=15, swing2LimitDegrees=15, twistLimitDegrees=0}
 {
  bone1Name=base_L_hair01
  bone2Name=head
  swing1LimitDegrees=45
  swing1Motion=Locked
  swing2LimitDegrees=45
  swing2Motion=Locked
  twistLimitDegrees=45
  twistMotion=Locked
 }
 {
  bone1Name=base_L_hair02
  bone2Name=head
  swing1LimitDegrees=45
  swing1Motion=Locked
  swing2LimitDegrees=45
  swing2Motion=Locked
  twistLimitDegrees=45
  twistMotion=Locked
 }
 {
  bone1Name=base_L_hair03
  bone2Name=head
  swing1LimitDegrees=45
  swing1Motion=Locked
  swing2LimitDegrees=45
  swing2Motion=Locked
  twistLimitDegrees=45
  twistMotion=Locked
 }
 {
  bone1Name=base_L_hair04
  bone2Name=head
  swing1LimitDegrees=45
  swing1Motion=Locked
  swing2LimitDegrees=45
  swing2Motion=Locked
  twistLimitDegrees=45
  twistMotion=Locked
 }
 {
  bone1Name=base_R_hair01
  bone2Name=head
  swing1LimitDegrees=45
  swing1Motion=Locked
  swing2LimitDegrees=45
  swing2Motion=Locked
  twistLimitDegrees=45
  twistMotion=Locked
 }
 {
  bone1Name=base_R_hair02
  bone2Name=head
  swing1LimitDegrees=45
  swing1Motion=Locked
  swing2LimitDegrees=45
  swing2Motion=Locked
  twistLimitDegrees=45
  twistMotion=Locked
 }
 {
  bone1Name=base_hair01
  bone2Name=head
  swing1LimitDegrees=45
  swing1Motion=Locked
  swing2LimitDegrees=45
  swing2Motion=Locked
  twistLimitDegrees=45
  twistMotion=Locked
 }
 {bone1Name=dyn_hair01, bone2Name=base_hair01, swing1LimitDegrees=10, swing2LimitDegrees=8, twistLimitDegrees=1}
 {bone1Name=dyn_hair02, bone2Name=base_hair01, swing1LimitDegrees=10, swing2LimitDegrees=8, twistLimitDegrees=1}
 {bone1Name=dyn_hair03, bone2Name=base_hair01, swing1LimitDegrees=10, swing2LimitDegrees=8, twistLimitDegrees=1}
 {bone1Name=dyn_hair04, bone2Name=base_hair01, swing1LimitDegrees=10, swing2LimitDegrees=8, twistLimitDegrees=1}
]
