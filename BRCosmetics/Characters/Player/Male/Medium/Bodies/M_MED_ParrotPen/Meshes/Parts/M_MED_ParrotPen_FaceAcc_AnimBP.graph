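# Animation Blueprint for the ParrotPen character's face accessory, adding movement-reactive hair motion on top of the pose copied from the attached parent mesh.
# Each update derives direction flags and eases a hair pitch driven by jumping and forward movement.
# Additive bone edits apply that pitch and a skydive-floating tilt to base_hair01, then feed a head-relative RigidBody simulation.
# A gravity override is computed but unread; the simulation uses fixed gravity.
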AnimBlueprint /BRCosmetics/Characters/Player/Male/Medium/Bodies/M_MED_ParrotPen/Meshes/Parts/M_MED_ParrotPen_FaceAcc_AnimBP
target_skeleton=&/M_MED_ParrotPen_FaceAcc_Skeleton
graphs [
 AnimGraph {nodes [
   n1 = "Transform (Modify) Bone - Bone: base_hair01"!(
    ComponentPose=Local To Component(LocalPose=Copy Pose From Mesh() {bUseAttachedParent=true})
    Alpha=1
    Rotation=Get Hair_Rot() -> (Hair_Rot:Rotator)
   ) -> (Pose) {
    rotationMode=BMM_Additive
    rotationSpace=BCS_BoneSpace
    alphaBoolBlend {blendOption=Linear}
    boneToModify {boneName=base_hair01}
    rotation {pitch=-90}
   }
   n2 = "Transform (Modify) Bone - Bone: base_hair01"(ComponentPose=@n1, Alpha=Get SkydiveFloating() -> (SkydiveFloating:double), Rotation="-15,0,0") {
    rotationMode=BMM_Additive
    rotationSpace=BCS_BoneSpace
    alphaBoolBlend {blendOption=Linear}
    boneToModify {boneName=base_hair01}
    rotation {pitch=-15}
   }
   n3 = RigidBody!(ComponentPose=@n2, Alpha, OverrideWorldGravity:Vector="0,0,400") -> (Pose) {
    bClampLinearTranslationLimitToRefPose=true
    bOverrideWorldGravity=true
    overridePhysicsAsset=&/M_MED_ParrotPen_FaceAcc_Physics
    simulationSpace=BaseBoneSpace
    alphaBoolBlend {blendInTime=0.25, blendOption=Linear, blendOutTime=0.25}
    baseBoneRef {boneName=head}
    componentLinearAccScale=<0.2 0.2 0.2>
    componentLinearVelScale=<0.2 0.1 0.1>
    overrideWorldGravity {z=980}
    simSpaceSettings {
     maxAngularAcceleration=1002
     maxAngularVelocity=3
     maxLinearAcceleration=700
     maxLinearVelocity=500
     velocityScaleZ=0
     worldAlpha=0.5
     externalLinearDragV=<0.2 0.2>
    }
   }
   Output Pose(Result=Component To Local(ComponentPose=@n3)) {layerGroup=None}
  ]}
 EventGraph {nodes [
   n1 = Event BlueprintUpdateAnimation()
   > n2 = Sequence!(execute) -> (then_0, then_1, then_2:Exec, then_3:Exec)
   > n3 = Gravity Override State Function DEPRECATED(CharacterPartAnimInstance=Self-Reference(), GameTimeInSeconds=GetGameTimeInSeconds(), GravityOverrideParams=Get Gravity Override Params())
   > Set GravityOverride(GravityOverride=@n3)
   > n4 = PlayerMovement(AnimInstance=Self-Reference(), XDotProductCoefficient=100, YDotProductCoefficient=100, ZDotProductCoefficient=100)
   n5 = BreakVector(InVec=MakeVector(X=@n4, Y=@n4.YMovement))
   > Set Left(Left:double="To Float (Boolean)"(InBool="float < float"(A=@n5, B=-50))) -> (Output_Get:double)
   > Set Right(Right:double="To Float (Boolean)"(InBool="float > float"(A=@n5, B=50))) -> (Output_Get:double)
   > Set Forward(Forward:double="To Float (Boolean)"(InBool="float > float"(A=@n5.Y, B=40))) -> (Output_Get:double)
   > Set Backward(Backward:double="To Float (Boolean)"(InBool="float < float"(A=@n5.Y, B=-40))) -> (Output_Get:double)
   n6 = float * float() {is=title K2Node_PromotableOperator}
   n7 = float * float(A=float + float(A:bool=Get bIsJumping(), B=Get Forward() -> (Forward:double)), B=-10)
   n8 = Set Hair_Y_Float(
    execute=@n2.then_1
    Hair_Y_Float:double=FInterpTo(Current=Get Hair_Y_Float() -> (Hair_Y_Float:double), Target=float + float(A:double=@n7, B=@n6), DeltaTime=@n1.DeltaTimeX, InterpSpeed=8)
   ) -> (Output_Get:double)
   > Set Hair_Rot(Hair_Rot:Rotator=MakeRotator(Pitch=@n8)) -> (Output_Get:Rotator)
   Set SkydiveFloating(execute=@n2.then_2, SkydiveFloating:double="To Float (Boolean)"(InBool=Get bIsSkydiveFloating())) -> (Output_Get:double)
   Get Backward() -> (Backward:double)
   Get Left() -> (Left:double)
   Get Right() -> (Right:double)
  ]}
]
variables [
 Backward {type=number}
 Forward {type=number}
 Gravity Override Params {type=GravityOverrideParamsStruct}
 GravityOverride {type=Vector}
 Hair_Rot {type=Rotator}
 Hair_Y_Float {type=number}
 Left {type=number}
 Right {type=number}
 SkydiveFloating {type=number}
]
